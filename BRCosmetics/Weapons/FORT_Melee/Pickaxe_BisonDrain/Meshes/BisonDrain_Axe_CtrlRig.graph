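# Control Rig for the frog character on the BisonDrain pickaxe, posing its bones procedurally.
# A debug-gated setup step, switched off, can drive root and spine from two controls, then fixed transform offsets adjust body, arms and mouth.
# Elbows and knees are parent-constrained to frog_root, legs aim at knees, and upper-leg twists follow knees on X translation only.
# A separate four-constraint arm chain is left unconnected.
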ControlRigBlueprint /BRCosmetics/Weapons/FORT_Melee/Pickaxe_BisonDrain/Meshes/BisonDrain_Axe_CtrlRig
graphs [
 New Function_ContainedGraph {nodes [
   n1 = RigVMFunctionEntryNode()
   RigVMFunction_ControlFlowBranch = &RigVMFunction_ControlFlowBranch(Condition=@n1.Debug, ExecuteContext=@n1.ExecuteContext)
   RigVMFunctionReturnNode(ExecuteContext=@RigVMFunction_ControlFlowBranch.Completed)
   &RigUnit_SetTransform(
    Item=(Type=Bone,Name="spine_01_Jnt")
    Weight=1
    bInitial=False
    Value=&RigUnit_GetTransform(Item=(Type=Control,Name="spine_01_Jnt_ctrl"), bInitial=False).Transform
    ExecutePin=@RigVMFunction_ControlFlowBranch.True
   )
   > &RigUnit_SetTransform(
    Item=(Type=Bone,Name="frog_root")
    Weight=1
    bInitial=False
    Value=&RigUnit_GetTransform(Item=(Type=Control,Name="frog_root_ctrl"), bInitial=False).Transform
   )
  ]}
 RigVMFunctionLibrary {nodes [ctrlSetup = RigVMCollapseNode(Debug=False, ExecuteContext=())]}
 RigVMModel {nodes [
   &RigUnit_ParentConstraint(
    AdvancedSettings=(InterpolationType=Average,RotationOrderForFilter=XZY)
    Child=(Type=Bone,Name="L_elbow_Jnt")
    ExecutePin=()
    Filter=(TranslationFilter=(bX=True,bY=True,bZ=True),RotationFilter=(bX=true,bY=true,bZ=true),ScaleFilter=(bX=true,bY=true,bZ=true))
    Parents=((Item=(Type=Bone,Name="frog_root"),Weight=1))
    Weight=1
    bMaintainOffset=true
   )
   > &RigUnit_ParentConstraint(
    AdvancedSettings=(InterpolationType=Average,RotationOrderForFilter=XZY)
    Child=(Type=Bone,Name="R_elbow_Jnt")
    Filter=(TranslationFilter=(bX=True,bY=True,bZ=True),RotationFilter=(bX=true,bY=true,bZ=true),ScaleFilter=(bX=true,bY=true,bZ=true))
    Parents=((Item=(Type=Bone,Name="frog_root"),Weight=1))
    Weight=1
    bMaintainOffset=true
   )
   > &RigUnit_ParentConstraint(
    AdvancedSettings=(InterpolationType=Average,RotationOrderForFilter=XZY)
    Child=(Type=Bone,Name="L_UpperArm_Twist")
    Filter=(TranslationFilter=(bX=True,bY=True,bZ=True),RotationFilter=(bX=true,bY=true,bZ=true),ScaleFilter=(bX=true,bY=true,bZ=true))
    Parents=((Item=(Type=Bone,Name="frog_root"),Weight=0),(Item=(Type=Bone,Name="L_elbow_Jnt"),Weight=1))
    Weight=1
    bMaintainOffset=true
   )
   > &RigUnit_ParentConstraint(
    AdvancedSettings=(InterpolationType=Average,RotationOrderForFilter=XZY)
    Child=(Type=Bone,Name="R_UpperArm_Twist")
    Filter=(TranslationFilter=(bX=True,bY=True,bZ=True),RotationFilter=(bX=true,bY=true,bZ=true),ScaleFilter=(bX=true,bY=true,bZ=true))
    Parents=((Item=(Type=Bone,Name="frog_root"),Weight=0),(Item=(Type=Bone,Name="R_elbow_Jnt"),Weight=1))
    Weight=1
    bMaintainOffset=true
   )
   RigUnit_BeginExecution = &RigUnit_BeginExecution()
   ctrlSetup = RigVMFunctionReferenceNode(Debug=false, ExecuteContext=@RigUnit_BeginExecution.ExecutePin)
   > &RigUnit_ModifyTransforms(
    ItemToModify=((Item=(Type=Bone,Name="frog_root"),Transform=(Rotation=(X=-0.1305,Y=0,Z=0,W=0.9914),Translation=(X=0,Y=0,Z=0),Scale3D=(X=1,Y=1,Z=1))))
    Weight=1
    WeightMinimum=0
   )
   > &RigUnit_ModifyTransforms(
    ItemToModify=((Item=(Type=Bone,Name="spine_01_Jnt"),Transform=(Rotation=(X=0.1822,Y=0,Z=0,W=-0.9833),Translation=(X=0,Y=-0.5,Z=0),Scale3D=(X=1,Y=1,Z=1))))
    Weight=1
    WeightMinimum=0
   )
   > &RigUnit_ParentConstraint(
    AdvancedSettings=(InterpolationType=Average,RotationOrderForFilter=XZY)
    Child=(Type=Bone,Name="L_elbow_Jnt")
    Filter=(TranslationFilter=(bX=True,bY=True,bZ=True),RotationFilter=(bX=true,bY=true,bZ=true),ScaleFilter=(bX=true,bY=true,bZ=true))
    Parents=((Item=(Type=Bone,Name="frog_root"),Weight=1))
    Weight=1
    bMaintainOffset=true
   )
   > &RigUnit_ParentConstraint(
    AdvancedSettings=(InterpolationType=Average,RotationOrderForFilter=XZY)
    Child=(Type=Bone,Name="R_elbow_Jnt")
    Filter=(TranslationFilter=(bX=True,bY=True,bZ=True),RotationFilter=(bX=true,bY=true,bZ=true),ScaleFilter=(bX=true,bY=true,bZ=true))
    Parents=((Item=(Type=Bone,Name="frog_root"),Weight=1))
    Weight=1
    bMaintainOffset=true
   )
   > &RigUnit_ModifyTransforms(
    ItemToModify=((Item=(Type=Bone,Name="L_UpperArm_Twist"),Transform=(Rotation=(X=0,Y=0,Z=0,W=-1),Translation=(X=0.6,Y=-0.2,Z=0),Scale3D=(X=1,Y=1,Z=1))))
    Weight=1
    WeightMinimum=0
   )
   > &RigUnit_ModifyTransforms(
    ItemToModify=((Item=(Type=Bone,Name="R_UpperArm_Twist"),Transform=(Rotation=(X=0,Y=0,Z=0,W=-1),Translation=(X=-0.6,Y=0.2,Z=0),Scale3D=(X=1,Y=1,Z=1))))
    Weight=1
    WeightMinimum=0
   )
   > &RigUnit_ModifyTransforms(
    ItemToModify=((Item=(Type=Bone,Name="L_elbow_Jnt"),Transform=(Rotation=(X=0,Y=0.2192,Z=0,W=0.9757),Translation=(X=3.1839,Y=-0.308,Z=-0.4109),Scale3D=(X=1,Y=1,Z=1))))
    Weight=1
    WeightMinimum=0
   )
   > &RigUnit_ModifyTransforms(
    ItemToModify=((Item=(Type=Bone,Name="R_elbow_Jnt"),Transform=(Rotation=(X=0,Y=0.2192,Z=0,W=0.9757),Translation=(X=-3.1839,Y=0.308,Z=0.4109),Scale3D=(X=1,Y=1,Z=1))))
    Weight=1
    WeightMinimum=0
   )
   > &RigUnit_ModifyTransforms(
    ItemToModify=((Item=(Type=Bone,Name="R_wrist_Jnt"),Transform=(Rotation=(X=0.1374,Y=-0.4639,Z=0.0528,W=0.8736),Translation=(X=0,Y=0,Z=0),Scale3D=(X=1,Y=1,Z=1))))
    Weight=1
    WeightMinimum=0
   )
   > &RigUnit_ModifyTransforms(
    ItemToModify=((Item=(Type=Bone,Name="L_wrist_Jnt"),Transform=(Rotation=(X=0.1374,Y=-0.4639,Z=0.0528,W=0.8736),Translation=(X=0,Y=0,Z=0),Scale3D=(X=1,Y=1,Z=1))))
    Weight=1
    WeightMinimum=0
   )
   > &RigUnit_ModifyTransforms(
    ItemToModify=((Item=(Type=Bone,Name="jaw_Jnt"),Transform=(Rotation=(X=0,Y=0.1709,Z=0,W=-0.9853),Translation=(X=0,Y=0,Z=0),Scale3D=(X=1,Y=1,Z=1))))
    Weight=1
    WeightMinimum=0
   )
   > &RigUnit_ModifyTransforms(
    ItemToModify=((Item=(Type=Bone,Name="C_LowerLip_Jnt"),Transform=(Rotation=(X=0,Y=0.1288,Z=0,W=0.9917),Translation=(X=0,Y=0,Z=0),Scale3D=(X=1,Y=1,Z=1))))
    Weight=1
    WeightMinimum=0
   )
   > &RigUnit_ModifyTransforms(
    ItemToModify=((Item=(Type=Bone,Name="R_LowerLip_Jnt"),Transform=(Rotation=(X=-0.1262,Y=0,Z=0,W=0.992),Translation=(X=0,Y=0.0702,Z=-0.2692),Scale3D=(X=1,Y=1,Z=1))))
    Weight=1
    WeightMinimum=0
   )
   > &RigUnit_ModifyTransforms(
    ItemToModify=((Item=(Type=Bone,Name="L_LowerLip_Jnt"),Transform=(Rotation=(X=-0.1262,Y=0,Z=0,W=0.992),Translation=(X=0,Y=-0.0702,Z=0.2692),Scale3D=(X=1,Y=1,Z=1))))
    Weight=1
    WeightMinimum=0
   )
   > &RigUnit_ParentConstraint(
    AdvancedSettings=(InterpolationType=Average,RotationOrderForFilter=XZY)
    Child=(Type=Bone,Name="L_knee_Jnt")
    Filter=(TranslationFilter=(bX=True,bY=True,bZ=True),RotationFilter=(bX=true,bY=true,bZ=true),ScaleFilter=(bX=true,bY=true,bZ=true))
    Parents=((Item=(Type=Bone,Name="frog_root"),Weight=1))
    Weight=1
    bMaintainOffset=true
   )
   > &RigUnit_ParentConstraint(
    AdvancedSettings=(InterpolationType=Average,RotationOrderForFilter=XZY)
    Child=(Type=Bone,Name="R_knee_Jnt")
    Filter=(TranslationFilter=(bX=True,bY=True,bZ=True),RotationFilter=(bX=true,bY=true,bZ=true),ScaleFilter=(bX=true,bY=true,bZ=true))
    Parents=((Item=(Type=Bone,Name="frog_root"),Weight=1))
    Weight=1
    bMaintainOffset=true
   )
   > &RigUnit_AimConstraintLocalSpaceOffset(
    Child=(Type=Bone,Name="L_leg_Jnt")
    Filter=(bX=True,bY=True,bZ=True)
    Parents=((Item=(Type=Bone,Name="L_knee_Jnt"),Weight=1))
    Weight=1
    WorldUp=(Target=(X=0,Y=0,Z=1),Kind=Direction,Space=(Type=None,Name="None"))
   )
   > &RigUnit_AimConstraintLocalSpaceOffset(
    Child=(Type=Bone,Name="R_leg_Jnt")
    Filter=(bX=True,bY=True,bZ=True)
    Parents=((Item=(Type=Bone,Name="R_knee_Jnt"),Weight=1))
    Weight=1
    WorldUp=(Target=(X=0,Y=0,Z=1),Kind=Direction,Space=(Type=None,Name="None"))
   )
   > &RigUnit_ParentConstraint(
    AdvancedSettings=(InterpolationType=Average,RotationOrderForFilter=XZY)
    Child=(Type=Bone,Name="L_UpperLeg_Twist")
    Filter=(TranslationFilter=(bX=True,bY=false,bZ=false),RotationFilter=(bX=false,bY=false,bZ=false),ScaleFilter=(bX=false,bY=false,bZ=false))
    Parents=((Item=(Type=Bone,Name="L_leg_Jnt"),Weight=0),(Item=(Type=Bone,Name="L_knee_Jnt"),Weight=1))
    Weight=1
    bMaintainOffset=true
   )
   > &RigUnit_ParentConstraint(
    AdvancedSettings=(InterpolationType=Average,RotationOrderForFilter=XZY)
    Child=(Type=Bone,Name="R_UpperLeg_Twist")
    Filter=(TranslationFilter=(bX=true,bY=false,bZ=false),RotationFilter=(bX=false,bY=false,bZ=false),ScaleFilter=(bX=false,bY=false,bZ=false))
    Parents=((Item=(Type=Bone,Name="R_leg_Jnt"),Weight=0),(Item=(Type=Bone,Name="R_knee_Jnt"),Weight=1))
    Weight=1
    bMaintainOffset=true
   )
   > &RigUnit_ParentConstraint(
    AdvancedSettings=(InterpolationType=Average,RotationOrderForFilter=XZY)
    Child=(Type=Bone,Name="L_knee_Jnt")
    Filter=(TranslationFilter=(bX=True,bY=True,bZ=True),RotationFilter=(bX=true,bY=true,bZ=true),ScaleFilter=(bX=true,bY=true,bZ=true))
    Parents=((Item=(Type=Bone,Name="frog_root"),Weight=1))
    Weight=1
    bMaintainOffset=true
   )
   > &RigUnit_ParentConstraint(
    AdvancedSettings=(InterpolationType=Average,RotationOrderForFilter=XZY)
    Child=(Type=Bone,Name="R_knee_Jnt")
    Filter=(TranslationFilter=(bX=True,bY=True,bZ=True),RotationFilter=(bX=true,bY=true,bZ=true),ScaleFilter=(bX=true,bY=true,bZ=true))
    Parents=((Item=(Type=Bone,Name="frog_root"),Weight=1))
    Weight=1
    bMaintainOffset=true
   )
  ]}
]
hierarchy [
 b_AR_Root {loc=<0 0 0>, rot=<0 0 0.7071 0.7071>}
 Melee_TwoHanded_Handle {parent=b_AR_Root, loc=<8.1477 -0.008 -29.0793>, rot=<0 0 0.7071 -0.7071>}
 bate_Jnt {parent=Melee_TwoHanded_Handle, loc=<9.0924 2.2951 41.5212>, rot=<0.1735 -0.9848 0 0>}
 frog_root {parent=Melee_TwoHanded_Handle, loc=<0.311 -1.6958 41.5793>, rot=<-0.1227 0.6964 -0.1227 -0.6964>}
 spine_01_Jnt {parent=frog_root, loc=<-0.1741 -0.7438 -9.0974>, rot=<-0.0082 0 0 -1>}
 spine_02_Jnt {parent=spine_01_Jnt, loc=<0 -6.6732 0>, rot=<0 0 0 1>}
 spine_03_Jnt {parent=spine_02_Jnt, loc=<0 -3.3428 0>, rot=<0 0 0 1>}
 L_clavicle_Jnt {parent=spine_03_Jnt, loc=<4.5816 1.6183 2.0875>, rot=<-0.2128 0.4514 -0.5453 -0.6735>}
 L_shoulder_Jnt {parent=L_clavicle_Jnt, loc=<0 0 0>, rot=<0 0 0 1>}
 L_elbow_Jnt {parent=L_shoulder_Jnt, loc=<2.5897 0 0>, rot=<0 0 -0.0425 -0.9991>}
 L_wrist_Jnt {parent=L_elbow_Jnt, loc=<2.5774 0 0>, rot=<0 0 0 1>}
 L_pinky_00_Jnt {parent=L_wrist_Jnt, loc=<0.4475 -0.3964 -0.0934>, rot=<0.0094 0.0004 0.1639 -0.9864>}
 L_pinky_01_Jnt {parent=L_pinky_00_Jnt, loc=<0.71 0 0>, rot=<-0.0325 -0.0203 0.0701 -0.9968>}
 L_pinky_02_Jnt {parent=L_pinky_01_Jnt, loc=<0.5351 0 0>, rot=<-0.0006 -0.0464 -0.0253 -0.9986>}
 L_ring_00_Jnt {parent=L_wrist_Jnt, loc=<0.4793 -0.064 -0.0897>, rot=<-0.0107 0.0298 -0.017 -0.9994>}
 L_ring_01_Jnt {parent=L_ring_00_Jnt, loc=<0.713 0 0>, rot=<-0.0251 -0.0446 0.0275 -0.9983>}
 L_ring_02_Jnt {parent=L_ring_01_Jnt, loc=<0.494 0 0>, rot=<0.005 -0.0336 -0.0165 -0.9993>}
 L_middle_00_Jnt {parent=L_wrist_Jnt, loc=<0.4904 0.3175 -0.048>, rot=<0.0054 -0.0165 -0.1654 -0.9861>}
 L_middle_01_Jnt {parent=L_middle_00_Jnt, loc=<0.711 0 0>, rot=<0.0653 -0.018 -0.0769 -0.9947>}
 L_middle_02_Jnt {parent=L_middle_01_Jnt, loc=<0.483 0 0>, rot=<-0.042 0.0375 0.0539 -0.997>}
 L_LowerArm_Twist {parent=L_elbow_Jnt, loc=<1.2887 0 0>, rot=<0 0 0 1>}
 L_UpperArm_Twist {parent=L_shoulder_Jnt, loc=<1.2948 0 0>, rot=<0 0 0 1>}
 neck_01_Jnt {parent=spine_03_Jnt, loc=<0 -0.0522 0.0036>, rot=<0.0245 -0.7067 0.0245 -0.7067>}
 neck_02_Jnt {parent=neck_01_Jnt, loc=<0 -0.0501 0>, rot=<0 0 0 1>}
 head_Jnt {parent=neck_02_Jnt, loc=<0 -0.2571 0>, rot=<0.0051 0.0063 -0.0163 -0.9998>}
 top_head_Jnt {parent=head_Jnt, loc=<3.753 -0.3133 -0.0122>, rot=<0.1929 -0.6849 0.6767 -0.1893>}
 L_eye_socket_Jnt {parent=top_head_Jnt, loc=<3.3796 -2.2675 0.8098>, rot=<-0.0515 0.8584 0.0327 -0.5094>}
 L_upperLid_Jnt {parent=L_eye_socket_Jnt, loc=<0.934 0 0>, rot=<0 0 0 1>}
 L_loweLid_Jnt {parent=L_eye_socket_Jnt, loc=<0.934 0 0>, rot=<0 0 0 1>}
 L_eye_Jnt {parent=L_eye_socket_Jnt, loc=<0.934 0 0>, rot=<0 0 0 1>}
 R_eye_socket_Jnt {parent=top_head_Jnt, loc=<3.3949 2.3248 0.7965>, rot=<0.0335 -0.5093 0.0558 -0.8581>}
 R_upperLid_Jnt {parent=R_eye_socket_Jnt, loc=<-0.934 0 0>, rot=<0 0 0 1>}
 R_rowerLid_Jnt {parent=R_eye_socket_Jnt, loc=<-0.934 0 0>, rot=<0 0 0 1>}
 R_eye_Jnt {parent=R_eye_socket_Jnt, loc=<-0.934 0 0>, rot=<0 0 0 1>}
 C_UpperLip_Jnt {parent=head_Jnt, loc=<-1.9228 -3.7093 -0.0151>, rot=<0.2709 -0.6579 0.6491 -0.2693>}
 L_UpperLip_Jnt {parent=head_Jnt, loc=<-0.5314 -1.9676 2.523>, rot=<0.5566 -0.4378 0.7021 0.0744>}
 L_CornerLip_Jnt {parent=head_Jnt, loc=<0.3488 -0.6737 3.411>, rot=<0.5218 -0.4796 0.6794 0.1898>}
 R_UpperLip_Jnt {parent=head_Jnt, loc=<-0.5942 -1.9144 -2.5498>, rot=<0.4313 0.5471 0.0748 -0.7135>}
 R_CornerLip_Jnt {parent=head_Jnt, loc=<0.264 -0.6019 -3.432>, rot=<0.4748 0.5113 0.1893 -0.6908>}
 jaw_Jnt {parent=head_Jnt, loc=<0.6538 -0.4043 0.0114>, rot=<0.1761 -0.6893 0.6813 -0.1723>}
 jaw_end_Jnt {parent=jaw_Jnt, loc=<4.2692 -0.1065 0.2117>, rot=<0 0 0 1>}
 C_LowerLip_Jnt {parent=jaw_Jnt, loc=<4.2712 0 0.208>, rot=<0 0 0 1>}
 L_LowerLip_Jnt {parent=jaw_Jnt, loc=<2.0564 -2.5511 -0.2026>, rot=<0.0311 0.1022 0.4956 -0.8619>}
 R_LowerLip_Jnt {parent=jaw_Jnt, loc=<2.0739 2.509 -0.2165>, rot=<0.4929 -0.8637 -0.0284 -0.101>}
 tongue_01_Jnt {parent=jaw_Jnt, loc=<0.1408 0.0217 -0.586>, rot=<0.0003 0.1315 -0.0004 -0.9913>}
 tongue_02_Jnt {parent=tongue_01_Jnt, loc=<1.6616 0 0>, rot=<0.0037 -0.0546 0.0045 -0.9985>}
 R_clavicle_Jnt {parent=spine_03_Jnt, loc=<-4.5816 1.6183 2.0875>, rot=<0.6735 -0.5453 -0.4514 -0.2128>}
 R_shoulder_Jnt {parent=R_clavicle_Jnt, loc=<0 0 0>, rot=<0 0 0 1>}
 R_elbow_Jnt {parent=R_shoulder_Jnt, loc=<-2.5897 0 0>, rot=<0 0 -0.0425 -0.9991>}
 R_wrist_Jnt {parent=R_elbow_Jnt, loc=<-2.5774 0 0>, rot=<0 0 0 1>}
 R_pinky_00_Jnt {parent=R_wrist_Jnt, loc=<-0.4475 0.3964 0.0934>, rot=<0.0094 0.0004 0.1639 -0.9864>}
 R_pinky_01_Jnt {parent=R_pinky_00_Jnt, loc=<-0.71 0 0>, rot=<-0.0325 -0.0203 0.0701 -0.9968>}
 R_pinky_02_Jnt {parent=R_pinky_01_Jnt, loc=<-0.5351 0 0>, rot=<-0.0006 -0.0464 -0.0253 -0.9986>}
 R_ring_00_Jnt {parent=R_wrist_Jnt, loc=<-0.4793 0.064 0.0897>, rot=<-0.0107 0.0298 -0.017 -0.9994>}
 R_ring_01_Jnt {parent=R_ring_00_Jnt, loc=<-0.713 0 0>, rot=<-0.0251 -0.0446 0.0275 -0.9983>}
 R_ring_02_Jnt {parent=R_ring_01_Jnt, loc=<-0.494 0 0>, rot=<0.005 -0.0336 -0.0165 -0.9993>}
 R_middle_00_Jnt {parent=R_wrist_Jnt, loc=<-0.4904 -0.3175 0.048>, rot=<0.0054 -0.0165 -0.1654 -0.9861>}
 R_middle_01_Jnt {parent=R_middle_00_Jnt, loc=<-0.711 0 0>, rot=<0.0653 -0.018 -0.0769 -0.9947>}
 R_middle_02_Jnt {parent=R_middle_01_Jnt, loc=<-0.483 0 0>, rot=<-0.042 0.0375 0.0539 -0.997>}
 R_LowerArm_Twist {parent=R_elbow_Jnt, loc=<-1.2887 0 0>, rot=<0 0 0 -1>}
 R_UpperArm_Twist {parent=R_shoulder_Jnt, loc=<-1.2948 0 0>, rot=<0 0 0 1>}
 L_leg_Jnt {parent=spine_01_Jnt, loc=<3.481 -0.5858 -0.7005>, rot=<0.9508 -0.1596 0.0756 0.2546>}
 L_knee_Jnt {parent=L_leg_Jnt, loc=<8.4736 0 0>, rot=<0 -0.7604 0 -0.6494>}
 L_ankle_Jnt {parent=L_knee_Jnt, loc=<4.5372 0 0>, rot=<0 0.664 0 -0.7478>}
 L_ball_Jnt {parent=L_ankle_Jnt, loc=<2.4792 0 0>, rot=<0 0.1418 0 -0.9899>}
 L_ball_FK_Jnt {parent=L_ball_Jnt, loc=<0 0 0>, rot=<0 0 0 1>}
 L_fingers_Jnt {parent=L_ball_FK_Jnt, loc=<1.11 0.1292 -0.153>, rot=<0 0 -0.0877 -0.9961>}
 L_LowerLeg_Twist {parent=L_knee_Jnt, loc=<2.2686 0 0>, rot=<0 0 0 1>}
 L_UpperLeg_Twist {parent=L_leg_Jnt, loc=<4.2368 0 0>, rot=<0 0 0 1>}
 R_leg_Jnt {parent=spine_01_Jnt, loc=<-3.481 -0.5858 -0.7005>, rot=<0.2546 -0.0756 -0.1596 -0.9508>}
 R_knee_Jnt {parent=R_leg_Jnt, loc=<-8.4736 0 0>, rot=<0 -0.7604 0 -0.6494>}
 R_ankle_Jnt {parent=R_knee_Jnt, loc=<-4.5373 0 0>, rot=<0 0.664 0 -0.7478>}
 R_ball_Jnt {parent=R_ankle_Jnt, loc=<-2.4792 0 0>, rot=<0 0.1418 0 -0.9899>}
 R_ball_FK_Jnt {parent=R_ball_Jnt, loc=<0 0 0>, rot=<0 0 0 1>}
 R_fingers_Jnt {parent=R_ball_FK_Jnt, loc=<-0.8923 -0.2085 0.1656>, rot=<0 0 -0.0877 -0.9961>}
 R_LowerLeg_Twist {parent=R_knee_Jnt, loc=<-2.2686 0 0>, rot=<0 0 0 -1>}
 R_UpperLeg_Twist {parent=R_leg_Jnt, loc=<-4.2368 0 0>, rot=<0 0 0 -1>}
 spine_01_Jnt_ctrl {controlType=EULER_TRANSFORM, type=Control, loc=<0 0 0>, rot=<0 0 0 1>}
 frog_root_ctrl {controlType=EULER_TRANSFORM, type=Control, loc=<0 0 0>, rot=<0 0 0 1>}
]
