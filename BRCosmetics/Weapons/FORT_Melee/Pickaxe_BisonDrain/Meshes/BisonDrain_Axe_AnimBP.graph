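# Anim Blueprint for the BisonDrain pickaxe, animating the weapon's creature rig with idle motion and secondary physics.
# A state machine leaves looping idle for a one-shot idle pose when a melee anim plays, then returns automatically, with DefaultSlot layered additively.
# Anim Dynamics on the spine, a Control Rig, then chained Anim Dynamics on jaw, tongue and eye sockets follow, all weighted by the DynamicsAlpha curve.
# The event graph stores IdleCurve in IdleAlpha, which no node reads.
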AnimBlueprint /BRCosmetics/Weapons/FORT_Melee/Pickaxe_BisonDrain/Meshes/BisonDrain_Axe_AnimBP
target_skeleton=&/BisonDrain_Axe_Skeleton
graphs [
 AnimGraph {nodes [
   n1 = Slot 'DefaultSlot'(Source=Local Space Ref Pose())
   AnimGraphNode_StateMachine_0 = Idle_State() {is=title AnimGraphNode_StateMachine}
   n2 = Anim Dynamics - Bone: spine_01_Jnt!(
    ComponentPose=Local To Component(LocalPose=Apply Additive(Base=@AnimGraphNode_StateMachine_0, Additive=@n1) {alphaBoolBlend {blendOption=Linear}})
    ExternalForce="0,0,0"
   ) -> (Pose) {
    alphaCurveName=DynamicsAlpha
    alphaInputType=Curve
    angularDampingOverride=1
    angularSpringConstant=10000
    bAngularSpring=true
    bLinearSpring=true
    bOverrideAngularDamping=true
    bUseGravityOverride=true
    gravityScale=0.35
    linearSpringConstant=20
    simulationSpace=BoneRelative
    alphaBoolBlend {blendOption=Linear}
    boundBone {boneName=spine_01_Jnt}
    componentAppliedLinearAccClamp=<3000 3000 3000>
    componentLinearAccScale=<0.3 0.3 0.3>
    componentLinearVelScale=<0.3 0.3 0.3>
    gravityOverride {z=-980}
    physicsBodyDefinitions [{
      boundBone {boneName=spine_01_Jnt}
      boxExtents=<40 60 60>
      constraintSetup {
       angularTargetAxis=AxisY
       twistAxis=AxisY
       angularLimitsMax=<0 2 2>
       angularLimitsMin=<0 -2 -2>
       angularTarget=<0 1 0>
       linearAxesMin=<0 -2 -1>
      }
      localJointOffset=<0 0 4>
     }]
    relativeSpaceBone {boneName=frog_root}
   }
   n3 = Control Rig!(Source=Component To Local(ComponentPose=@n2)) -> (Pose) {
    alphaCurveName=DynamicsAlpha
    alphaInputType=Curve
    alphaBoolBlend {blendOption=Linear}
    controlRigAssetReference {blueprintRigClass=&/BisonDrain_Axe_CtrlRig.BisonDrain_Axe_CtrlRig_C}
   }
   n4 = Anim Dynamics - Bone: jaw_Jnt!(ComponentPose=Local To Component(LocalPose=@n3), ExternalForce="0,0,0") -> (Pose) {
    alphaCurveName=DynamicsAlpha
    alphaInputType=Curve
    angularDampingOverride=1
    angularSpringConstant=10000
    bAngularSpring=true
    bOverrideAngularDamping=true
    bUseGravityOverride=true
    gravityScale=0.35
    simulationSpace=RootRelative
    alphaBoolBlend {blendOption=Linear}
    boundBone {boneName=jaw_Jnt}
    componentAppliedLinearAccClamp=<3000 3000 3000>
    componentLinearAccScale=<1 1 1>
    componentLinearVelScale=<1 1 1>
    gravityOverride {z=-980}
    physicsBodyDefinitions [{
      boundBone {boneName=jaw_Jnt}
      boxExtents=<20 30 30>
      constraintSetup {angularLimitsMax=<0 50 0>, angularLimitsMin=<0 0 0>, angularTarget=<1 0 0>}
      localJointOffset=<4 0 0>
     }]
   }
   n5 = Anim Dynamics - Bone: tongue_01_Jnt!(ComponentPose=@n4, ExternalForce="0,0,0") -> (Pose) {
    alphaCurveName=DynamicsAlpha
    alphaInputType=Curve
    angularDampingOverride=1
    angularSpringConstant=10000
    bAngularSpring=true
    bOverrideAngularDamping=true
    bUseGravityOverride=true
    gravityScale=0.35
    simulationSpace=RootRelative
    alphaBoolBlend {blendOption=Linear}
    boundBone {boneName=tongue_01_Jnt}
    componentAppliedLinearAccClamp=<3000 3000 3000>
    componentLinearAccScale=<1 1 1>
    componentLinearVelScale=<1 1 1>
    gravityOverride {z=-980}
    physicsBodyDefinitions [{
      boundBone {boneName=tongue_01_Jnt}
      boxExtents=<20 30 30>
      constraintSetup {angularLimitsMax=<0 50 0>, angularLimitsMin=<0 -3 0>, angularTarget=<1 0 0>}
      localJointOffset=<4 0 0>
     }]
   }
   n6 = Anim Dynamics - Bone: tongue_02_Jnt!(ComponentPose=@n5, ExternalForce="0,0,0") -> (Pose) {
    alphaCurveName=DynamicsAlpha
    alphaInputType=Curve
    angularDampingOverride=1
    angularSpringConstant=10000
    bAngularSpring=true
    bOverrideAngularDamping=true
    bUseGravityOverride=true
    gravityScale=0.35
    simulationSpace=RootRelative
    alphaBoolBlend {blendOption=Linear}
    boundBone {boneName=tongue_02_Jnt}
    componentAppliedLinearAccClamp=<3000 3000 3000>
    componentLinearAccScale=<1 1 1>
    componentLinearVelScale=<1 1 1>
    gravityOverride {z=-980}
    physicsBodyDefinitions [{
      boundBone {boneName=tongue_02_Jnt}
      boxExtents=<20 30 30>
      constraintSetup {angularLimitsMax=<0 50 0>, angularLimitsMin=<0 -3 0>, angularTarget=<1 0 0>}
      localJointOffset=<4 0 0>
     }]
   }
   n7 = Anim Dynamics - Bone: L_eye_socket_Jnt!(ComponentPose=@n6, ExternalForce="0,0,0") -> (Pose) {
    alphaCurveName=DynamicsAlpha
    alphaInputType=Curve
    angularDampingOverride=1
    angularSpringConstant=10000
    bAngularSpring=true
    bOverrideAngularDamping=true
    bUseGravityOverride=true
    gravityScale=0.35
    simulationSpace=RootRelative
    alphaBoolBlend {blendOption=Linear}
    boundBone {boneName=L_eye_socket_Jnt}
    componentAppliedLinearAccClamp=<3000 3000 3000>
    componentLinearAccScale=<1 1 1>
    componentLinearVelScale=<1 1 1>
    gravityOverride {z=-980}
    physicsBodyDefinitions [{
      boundBone {boneName=L_eye_socket_Jnt}
      boxExtents=<20 30 30>
      constraintSetup {angularLimitsMax=<0 0 3>, angularLimitsMin=<0 0 -3>, angularTarget=<1 0 0>}
      localJointOffset=<4 0 0>
     }]
   }
   n8 = Anim Dynamics - Bone: R_eye_socket_Jnt!(ComponentPose=@n7, ExternalForce="0,0,0") -> (Pose) {
    alphaCurveName=DynamicsAlpha
    alphaInputType=Curve
    angularDampingOverride=1
    angularSpringConstant=10000
    bAngularSpring=true
    bOverrideAngularDamping=true
    bUseGravityOverride=true
    gravityScale=0.35
    simulationSpace=RootRelative
    alphaBoolBlend {blendOption=Linear}
    boundBone {boneName=R_eye_socket_Jnt}
    componentAppliedLinearAccClamp=<3000 3000 3000>
    componentLinearAccScale=<1 1 1>
    componentLinearVelScale=<1 1 1>
    gravityOverride {z=-980}
    physicsBodyDefinitions [{
      boundBone {boneName=R_eye_socket_Jnt}
      boxExtents=<20 30 30>
      constraintSetup {angularLimitsMax=<0 0 3>, angularLimitsMin=<0 0 -3>, angularTarget=<1 0 0>}
      localJointOffset=<4 0 0>
     }]
   }
   Output Pose(Result=Component To Local(ComponentPose=@n8)) {layerGroup=None}
  ]}
 Idle_State {
  entry=Idle
  path=AnimGraph.AnimGraphNode_StateMachine_0.Idle_State
  ids=id
  nodes [
   AnimStateNode_0 = Idle()
   AnimStateNode_1 = Swinging() {is=title AnimStateNode}
   AnimStateTransitionNode_0 = AnimStateTransitionNode() {blendMode=Linear, crossfadeDuration=0}
   AnimStateTransitionNode_2 = AnimStateTransitionNode() {bAutomaticRuleBasedOnSequencePlayerInState=true, blendMode=Linear, crossfadeDuration=0}
  ]
  links [
   {
    dst=Swinging
    rule=AnimGraph.AnimGraphNode_StateMachine_0.Idle_State.AnimStateTransitionNode_0.Transition
    src=Idle
    via=AnimStateTransitionNode_0
   }
   {
    dst=Idle
    rule=AnimGraph.AnimGraphNode_StateMachine_0.Idle_State.AnimStateTransitionNode_2.Transition
    src=Swinging
    via=AnimStateTransitionNode_2
   }
  ]
 }
 Idle {
  path=AnimGraph.AnimGraphNode_StateMachine_0.Idle_State.AnimStateNode_0.Idle
  nodes [Output Animation Pose(Result=Sequence Player 'BisonDrain_Pickaxe_Idle_WEP'() {sequence=/BRCosmetics/Weapons/FORT_Melee/Pickaxe_BisonDrain/Animation/BisonDrain_Pickaxe_Idle_WEP})]
 }
 Swinging {
  path=AnimGraph.AnimGraphNode_StateMachine_0.Idle_State.AnimStateNode_1.Swinging
  nodes [Output Animation Pose(Result=Sequence Player 'BisonDrain_Pickaxe_IdlePose_WEP'() {bLoopAnimation=false, sequence=/BRCosmetics/Weapons/FORT_Melee/Pickaxe_BisonDrain/Animation/BisonDrain_Pickaxe_IdlePose_WEP})]
 }
 Transition {
  path=AnimGraph.AnimGraphNode_StateMachine_0.Idle_State.AnimStateTransitionNode_0.Transition
  nodes [Result(bCanEnterTransition=Get bIsPlayingMeleeAnim())]
 }
 Transition {path=AnimGraph.AnimGraphNode_StateMachine_0.Idle_State.AnimStateTransitionNode_2.Transition, nodes [Result()]}
 EventGraph {nodes [
   Event BlueprintUpdateAnimation()
   > Set IdleAlpha(IdleAlpha:double=GetCurveValue(CurveName=IdleCurve)) -> (Output_Get:double)
   TryGetPawnOwner()
  ]}
]
variables [IdleAlpha {type=number}]
